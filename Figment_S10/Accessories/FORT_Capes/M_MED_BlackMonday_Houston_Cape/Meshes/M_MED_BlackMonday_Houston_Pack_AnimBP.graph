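# Animation Blueprint for the BlackMonday Houston cape backpack, posing and simulating the cape across grounded, skydiving and DBNO states.
# It copies the attached parent's pose, adds a state-machine additive, then runs clavicle RBF Pose Drivers per state.
# Chained Trail controllers simulate four cape chains; grounded adds run/sprint additives, shoulder corrective offsets, fake velocities and planar limits.
# Cached results are selected by bIsDBNO and bIsSkyDiving.
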AnimBlueprint /Figment_S10/Accessories/FORT_Capes/M_MED_BlackMonday_Houston_Cape/Meshes/M_MED_BlackMonday_Houston_Pack_AnimBP
target_skeleton=/Game/Accessories/FORT_Capes/M_MED_BlackMonday_Houston_Cape/Meshes/M_MED_BlackMonday_Houston_Pack_Skeleton
graphs [
 AnimGraph {nodes [
   n1 = Use cached pose 'skydive_post_sim'()
   n2 = Use cached pose 'grounded_motion_cache'()
   n3 = Get bIsSkyDiving()
   Output Pose(Result=Blend Poses by bool(
     BlendPose_0=Use cached pose 'dbno_post_sim'()
     BlendPose_1=Blend Poses by bool(BlendPose_0=@n1, BlendPose_1=@n2, bActiveValue=@n3) {blendType=Linear}
     bActiveValue=Get bIsDBNO()
    ) {blendType=Linear}) {layerGroup=None}
   n4 = Get dbno_trail_alpha() -> (dbno_trail_alpha:double)
   n5 = Trail controller - Bone: cape_mid_07_l!(
    ComponentPose=Local To Component(LocalPose=Use cached pose 'dbno_cache'())
    Alpha=@n4
    RelaxationSpeedScale:float=2
    FakeVelocity:Vector="0,0,0"
   ) -> (Pose) {
    bActorSpaceFakeVel=true
    bShowTrailLocation=true
    bUsePlanarLimit=true
    chainLength=6
    debugLifeTime=0.1
    maxDeltaTime=0.0333
    relaxationSpeedScale=1.2
    alphaScaleBiasClamp {bInterpResult=true, interpSpeedDecreasing=25, interpSpeedIncreasing=25}
    baseJoint {boneName=root}
    rotationLimits [
     {limitMax=<180 180 180>}
     {limitMax=<180 180 180>}
     {limitMax=<180 180 180>}
     {limitMax=<180 180 180>}
     {limitMax=<180 180 180>}
     {limitMax=<180 180 180>}
    ]
    rotationOffsets [<0 0 0>, <0 0 0>, <0 0 0>, <0 0 0>, <0 0 0>, <0 0 0>]
    trailBone {boneName=cape_mid_07_l}
    trailRelaxationSpeed {editorCurveData {keys [<at 0 15 RCIM_Linear>,<at 1 6 RCIM_Linear>]}}
   }
   n6 = Trail controller - Bone: cape_mid_07_r!(ComponentPose=@n5, Alpha=@n4, RelaxationSpeedScale:float=2, FakeVelocity:Vector="0,0,0") -> (Pose) {
    bActorSpaceFakeVel=true
    bShowTrailLocation=true
    bUsePlanarLimit=true
    chainLength=6
    debugLifeTime=0.1
    maxDeltaTime=0.0333
    relaxationSpeedScale=1.2
    alphaScaleBiasClamp {bInterpResult=true, interpSpeedDecreasing=25, interpSpeedIncreasing=25}
    baseJoint {boneName=root}
    rotationLimits [
     {limitMax=<180 180 180>}
     {limitMax=<180 180 180>}
     {limitMax=<180 180 180>}
     {limitMax=<180 180 180>}
     {limitMax=<180 180 180>}
     {limitMax=<180 180 180>}
    ]
    rotationOffsets [<0 0 0>, <0 0 0>, <0 0 0>, <0 0 0>, <0 0 0>, <0 0 0>]
    trailBone {boneName=cape_mid_07_r}
    trailRelaxationSpeed {editorCurveData {keys [<at 0 15 RCIM_Linear>,<at 1 6 RCIM_Linear>]}}
   }
   n7 = Trail controller - Bone: cape_08_l!(ComponentPose=@n6, Alpha=@n4, RelaxationSpeedScale:float=2, FakeVelocity:Vector="0,0,0") -> (Pose) {
    bActorSpaceFakeVel=true
    bShowTrailLocation=true
    bUsePlanarLimit=true
    chainLength=6
    debugLifeTime=0.1
    maxDeltaTime=0.0333
    relaxationSpeedScale=1.2
    alphaScaleBiasClamp {bInterpResult=true, interpSpeedDecreasing=25, interpSpeedIncreasing=25}
    baseJoint {boneName=root}
    rotationLimits [
     {limitMax=<180 180 180>}
     {limitMax=<180 180 180>}
     {limitMax=<180 180 180>}
     {limitMax=<180 180 180>}
     {limitMax=<180 180 180>}
     {limitMax=<180 180 180>}
    ]
    rotationOffsets [<0 0 0>, <0 0 0>, <0 0 0>, <0 0 0>, <0 0 0>, <0 0 0>]
    trailBone {boneName=cape_08_l}
    trailRelaxationSpeed {editorCurveData {keys [<at 0 15 RCIM_Linear>,<at 1 6 RCIM_Linear>]}}
   }
   n8 = Trail controller - Bone: cape_08_r!(ComponentPose=@n7, Alpha=@n4, RelaxationSpeedScale:float=2, FakeVelocity:Vector="0,0,0") -> (Pose) {
    bActorSpaceFakeVel=true
    bShowTrailLocation=true
    bUsePlanarLimit=true
    chainLength=6
    debugLifeTime=0.1
    maxDeltaTime=0.0333
    relaxationSpeedScale=1.2
    alphaScaleBiasClamp {bInterpResult=true, interpSpeedDecreasing=25, interpSpeedIncreasing=25}
    baseJoint {boneName=root}
    rotationLimits [
     {limitMax=<180 180 180>}
     {limitMax=<180 180 180>}
     {limitMax=<180 180 180>}
     {limitMax=<180 180 180>}
     {limitMax=<180 180 180>}
     {limitMax=<180 180 180>}
    ]
    rotationOffsets [<0 0 0>, <0 0 0>, <0 0 0>, <0 0 0>, <0 0 0>, <0 0 0>]
    trailBone {boneName=cape_08_r}
    trailRelaxationSpeed {editorCurveData {keys [<at 0 15 RCIM_Linear>,<at 1 6 RCIM_Linear>]}}
   }
   Save cached pose 'dbno_post_sim'(Pose=Component To Local(ComponentPose=@n8)) {cacheName=dbno_post_sim}
   n9 = Get skydive_trail_alpha() -> (skydive_trail_alpha:double)
   n10 = Trail controller - Bone: cape_mid_07_l(
    ComponentPose=Local To Component(LocalPose=Use cached pose 'skydive_pre_sim_cache'())
    Alpha=@n9
    RelaxationSpeedScale=3
    FakeVelocity="0,0,0"
   ) {
    bActorSpaceFakeVel=true
    bShowTrailLocation=true
    bUsePlanarLimit=true
    chainLength=6
    debugLifeTime=0.1
    maxDeltaTime=0.0333
    relaxationSpeedScale=1.2
    alphaScaleBiasClamp {bInterpResult=true, interpSpeedDecreasing=25, interpSpeedIncreasing=25}
    baseJoint {boneName=root}
    rotationLimits [
     {limitMax=<180 180 180>}
     {limitMax=<180 180 180>}
     {limitMax=<180 180 180>}
     {limitMax=<180 180 180>}
     {limitMax=<180 180 180>}
     {limitMax=<180 180 180>}
    ]
    rotationOffsets [<0 0 0>, <0 0 0>, <0 0 0>, <0 0 0>, <0 0 0>, <0 0 0>]
    trailBone {boneName=cape_mid_07_l}
    trailRelaxationSpeed {editorCurveData {keys [<at 0 15 RCIM_Linear>,<at 1 6 RCIM_Linear>]}}
   }
   n11 = Trail controller - Bone: cape_mid_07_r(ComponentPose=@n10, Alpha=@n9, RelaxationSpeedScale=3, FakeVelocity="0,0,0") {
    bActorSpaceFakeVel=true
    bShowTrailLocation=true
    bUsePlanarLimit=true
    chainLength=6
    debugLifeTime=0.1
    maxDeltaTime=0.0333
    relaxationSpeedScale=1.2
    alphaScaleBiasClamp {bInterpResult=true, interpSpeedDecreasing=25, interpSpeedIncreasing=25}
    baseJoint {boneName=root}
    rotationLimits [
     {limitMax=<180 180 180>}
     {limitMax=<180 180 180>}
     {limitMax=<180 180 180>}
     {limitMax=<180 180 180>}
     {limitMax=<180 180 180>}
     {limitMax=<180 180 180>}
    ]
    rotationOffsets [<0 0 0>, <0 0 0>, <0 0 0>, <0 0 0>, <0 0 0>, <0 0 0>]
    trailBone {boneName=cape_mid_07_r}
    trailRelaxationSpeed {editorCurveData {keys [<at 0 15 RCIM_Linear>,<at 1 6 RCIM_Linear>]}}
   }
   n12 = Trail controller - Bone: cape_08_l(ComponentPose=@n11, Alpha=@n9, RelaxationSpeedScale=3, FakeVelocity="0,0,0") {
    bActorSpaceFakeVel=true
    bShowTrailLocation=true
    bUsePlanarLimit=true
    chainLength=6
    debugLifeTime=0.1
    maxDeltaTime=0.0333
    relaxationSpeedScale=1.2
    alphaScaleBiasClamp {bInterpResult=true, interpSpeedDecreasing=25, interpSpeedIncreasing=25}
    baseJoint {boneName=root}
    rotationLimits [
     {limitMax=<180 180 180>}
     {limitMax=<180 180 180>}
     {limitMax=<180 180 180>}
     {limitMax=<180 180 180>}
     {limitMax=<180 180 180>}
     {limitMax=<180 180 180>}
    ]
    rotationOffsets [<0 0 0>, <0 0 0>, <0 0 0>, <0 0 0>, <0 0 0>, <0 0 0>]
    trailBone {boneName=cape_08_l}
    trailRelaxationSpeed {editorCurveData {keys [<at 0 15 RCIM_Linear>,<at 1 6 RCIM_Linear>]}}
   }
   n13 = Trail controller - Bone: cape_08_r(ComponentPose=@n12, Alpha=@n9, RelaxationSpeedScale=3, FakeVelocity="0,0,0") {
    bActorSpaceFakeVel=true
    bShowTrailLocation=true
    bUsePlanarLimit=true
    chainLength=6
    debugLifeTime=0.1
    maxDeltaTime=0.0333
    relaxationSpeedScale=1.2
    alphaScaleBiasClamp {bInterpResult=true, interpSpeedDecreasing=25, interpSpeedIncreasing=25}
    baseJoint {boneName=root}
    rotationLimits [
     {limitMax=<180 180 180>}
     {limitMax=<180 180 180>}
     {limitMax=<180 180 180>}
     {limitMax=<180 180 180>}
     {limitMax=<180 180 180>}
     {limitMax=<180 180 180>}
    ]
    rotationOffsets [<0 0 0>, <0 0 0>, <0 0 0>, <0 0 0>, <0 0 0>, <0 0 0>]
    trailBone {boneName=cape_08_r}
    trailRelaxationSpeed {editorCurveData {keys [<at 0 15 RCIM_Linear>,<at 1 6 RCIM_Linear>]}}
   }
   Save cached pose 'skydive_post_sim'(Pose=Component To Local(ComponentPose=@n13)) {cacheName=skydive_post_sim}
   n14 = Get skydive_trail_alpha()
   n15 = BlackMonday_Houston_Poses_PoseAsset : Run!(PoseWeight:float=1) -> (Pose:Pose Link Structure) {
    is=title AnimGraphNode_PoseByName
    poseAsset=/Figment_S10/Accessories/FORT_Capes/M_MED_BlackMonday_Houston_Cape/Animation/BlackMonday_Houston_Poses_PoseAsset
    poseName=Run
   }
   n16 = BlackMonday_Houston_Poses_PoseAsset : Sprint!(PoseWeight:float=0.2) -> (Pose:Pose Link Structure) {
    is=title AnimGraphNode_PoseByName
    poseAsset=&BlackMonday_Houston_Poses_PoseAsset
    poseName=Sprint
   }
   n17 = Get SprintSpeedRemap() -> (SprintSpeedRemap:double)
   n18 = Two Way Blend(
    A=Apply Additive(Base=Use cached pose 'grounded_pre_sim_cache'(), Additive=@n16, Alpha=@n17) {alphaBoolBlend {blendOption=Linear}, alphaScaleBiasClamp {bInterpResult=true, interpSpeedDecreasing=5}}
    B=Apply Additive(Base=Use cached pose 'grounded_pre_sim_cache'(), Additive=@n15, Alpha=Get RunSpeedRemap() -> (RunSpeedRemap:double)) {alphaBoolBlend {blendOption=Linear}, alphaScaleBiasClamp {bInterpResult=true, interpSpeedDecreasing=5}}
    Alpha=@n17
   ) {blendNode {
     alphaBoolBlend {blendOption=Linear}
     alphaScaleBiasClamp {bInterpResult=true, bMapRange=true, interpSpeedDecreasing=5, outRange {max=0, min=1}}
    }}
   n19 = Use cached pose 'grounded_pre_sim_cache'()
   n20 = Get bIsCrouching()
   n21 = Get shoulder_corrective_alpha() -> (shoulder_corrective_alpha:double)
   n22 = "Transform (Modify) Bone - Bone: backpack_root"(
    ComponentPose=Local To Component(LocalPose=Blend Poses by bool(BlendPose_0=@n19, BlendPose_1=@n18, BlendTime_0=0.3, BlendTime_1=0.3, bActiveValue=@n20) {blendType=Linear})
    Alpha=@n21
    Translation="0,0,0"
    Rotation="0.000,0.000,17.86"
    Scale="1,1,1"
   ) {
    rotationMode=BMM_Additive
    rotationSpace=BCS_BoneSpace
    translationMode=BMM_Additive
    translationSpace=BCS_BoneSpace
    alphaBoolBlend {blendOption=Linear}
    alphaScaleBiasClamp {bInterpResult=true}
    boneToModify {boneName=backpack_root}
    rotation {roll=17.86}
   }
   n23 = "Transform (Modify) Bone - Bone: cape_01_l"(ComponentPose=@n22, Alpha=@n21, Translation="0,0,0", Rotation="5,-20.000,0.000", Scale="1,1,1") {
    rotationMode=BMM_Additive
    rotationSpace=BCS_BoneSpace
    alphaBoolBlend {blendOption=Linear}
    alphaScaleBiasClamp {bInterpResult=true}
    boneToModify {boneName=cape_01_l}
    rotation {pitch=5, yaw=-20}
   }
   n24 = "Transform (Modify) Bone - Bone: cape_01_r"(ComponentPose=@n23, Alpha=@n21, Translation="0,0,0", Rotation="-5,-20,0", Scale="1,1,1") {
    rotationMode=BMM_Additive
    rotationSpace=BCS_BoneSpace
    alphaBoolBlend {blendOption=Linear}
    alphaScaleBiasClamp {bInterpResult=true}
    boneToModify {boneName=cape_01_r}
    rotation {pitch=-5, yaw=-20}
   }
   n25 = Trail controller - Bone: cape_mid_07_l(
    ComponentPose=@n24
    Alpha=@n14
    RelaxationSpeedScale=Get relax_speed() -> (relax_speed:double)
    FakeVelocity=Get fake_velocity_B() -> (fake_velocity_B:Vector)
   ) {
    bActorSpaceFakeVel=true
    bShowTrailLocation=true
    bUsePlanarLimit=true
    chainLength=6
    debugLifeTime=0.1
    maxDeltaTime=0.0333
    relaxationSpeedScale=1.2
    alphaScaleBiasClamp {bInterpResult=true, interpSpeedDecreasing=25, interpSpeedIncreasing=25}
    baseJoint {boneName=root}
    planarLimits [
     {drivingBone {boneName=root}, planeTransform {location=<0 0 10>, rotation=<rot 0 0 0>, scale=<1 1 1>}}
     {drivingBone {boneName=spine_03}, planeTransform {location=<0 -20 0>, rotation=<rot 0 15 -90>, scale=<1 1 1>}}
    ]
    rotationLimits [
     {limitMax=<180 180 180>}
     {limitMax=<180 180 180>}
     {limitMax=<180 180 180>}
     {limitMax=<180 180 180>}
     {limitMax=<180 180 180>}
     {limitMax=<180 180 180>}
    ]
    rotationOffsets [<0 0 0>, <0 0 0>, <0 0 0>, <0 0 0>, <0 0 0>, <0 0 0>]
    trailBone {boneName=cape_mid_07_l}
    trailRelaxationSpeed {editorCurveData {keys [<at 0 15 RCIM_Linear>,<at 0.5289 13.1838 RCIM_Linear>,<at 1 6 RCIM_Linear>]}}
   }
   n26 = Trail controller - Bone: cape_mid_07_r(ComponentPose=@n25, Alpha=@n14, RelaxationSpeedScale=Get relax_speed(), FakeVelocity=Get fake_velocity_C() -> (fake_velocity_C:Vector)) {
    bActorSpaceFakeVel=true
    bShowTrailLocation=true
    bUsePlanarLimit=true
    chainLength=6
    debugLifeTime=0.1
    maxDeltaTime=0.0333
    relaxationSpeedScale=1.2
    alphaScaleBiasClamp {bInterpResult=true, interpSpeedDecreasing=25, interpSpeedIncreasing=25}
    baseJoint {boneName=root}
    planarLimits [
     {drivingBone {boneName=root}, planeTransform {location=<0 0 10>, rotation=<rot 0 0 0>, scale=<1 1 1>}}
     {drivingBone {boneName=spine_03}, planeTransform {location=<0 -20 0>, rotation=<rot 0 15 -89.9999>, scale=<1 1 1>}}
    ]
    rotationLimits [
     {limitMax=<180 180 180>}
     {limitMax=<180 180 180>}
     {limitMax=<180 180 180>}
     {limitMax=<180 180 180>}
     {limitMax=<180 180 180>}
     {limitMax=<180 180 180>}
    ]
    rotationOffsets [<0 0 0>, <0 0 0>, <0 0 0>, <0 0 0>, <0 0 0>, <0 0 0>]
    trailBone {boneName=cape_mid_07_r}
    trailRelaxationSpeed {editorCurveData {keys [<at 0 15 RCIM_Linear>,<at 0.5289 13.1838 RCIM_Linear>,<at 1 6 RCIM_Linear>]}}
   }
   n27 = Trail controller - Bone: cape_08_l(ComponentPose=@n26, Alpha=@n14, RelaxationSpeedScale=Get relax_speed(), FakeVelocity=Get fake_velocity_A() -> (fake_velocity_A:Vector)) {
    bActorSpaceFakeVel=true
    bShowTrailLocation=true
    bUsePlanarLimit=true
    chainLength=6
    debugLifeTime=0.1
    maxDeltaTime=0.0333
    relaxationSpeedScale=1.2
    alphaScaleBiasClamp {bInterpResult=true, interpSpeedDecreasing=25, interpSpeedIncreasing=25}
    baseJoint {boneName=root}
    planarLimits [
     {drivingBone {boneName=root}, planeTransform {location=<0 0 10>, rotation=<rot 0 0 0>, scale=<1 1 1>}}
     {drivingBone {boneName=spine_03}, planeTransform {location=<0 -20 0>, rotation=<rot 0 15 -89.9999>, scale=<1 1 1>}}
    ]
    rotationLimits [
     {limitMax=<180 180 180>}
     {limitMax=<180 180 180>}
     {limitMax=<180 180 180>}
     {limitMax=<180 180 180>}
     {limitMax=<180 180 180>}
     {limitMax=<180 180 180>}
    ]
    rotationOffsets [<0 0 0>, <0 0 0>, <0 0 0>, <0 0 0>, <0 0 0>, <0 0 0>]
    trailBone {boneName=cape_08_l}
    trailRelaxationSpeed {editorCurveData {keys [<at 0 15 RCIM_Linear>,<at 0.5289 13.1838 RCIM_Linear>,<at 1 6 RCIM_Linear>]}}
   }
   n28 = Trail controller - Bone: cape_08_r(ComponentPose=@n27, Alpha=@n14, RelaxationSpeedScale=Get relax_speed(), FakeVelocity=Get fake_velocity_D() -> (fake_velocity_D:Vector)) {
    bActorSpaceFakeVel=true
    bShowTrailLocation=true
    bUsePlanarLimit=true
    chainLength=6
    debugLifeTime=0.1
    maxDeltaTime=0.0333
    relaxationSpeedScale=1.2
    alphaScaleBiasClamp {bInterpResult=true, interpSpeedDecreasing=25, interpSpeedIncreasing=25}
    baseJoint {boneName=root}
    planarLimits [
     {drivingBone {boneName=root}, planeTransform {location=<0 0 10>, rotation=<rot 0 0 0>, scale=<1 1 1>}}
     {drivingBone {boneName=spine_03}, planeTransform {location=<0 -20 0>, rotation=<rot 0 15 -89.9999>, scale=<1 1 1>}}
    ]
    rotationLimits [
     {limitMax=<180 180 180>}
     {limitMax=<180 180 180>}
     {limitMax=<180 180 180>}
     {limitMax=<180 180 180>}
     {limitMax=<180 180 180>}
     {limitMax=<180 180 180>}
    ]
    rotationOffsets [<0 0 0>, <0 0 0>, <0 0 0>, <0 0 0>, <0 0 0>, <0 0 0>]
    trailBone {boneName=cape_08_r}
    trailRelaxationSpeed {editorCurveData {keys [<at 0 15 RCIM_Linear>,<at 0.5289 13.1838 RCIM_Linear>,<at 1 6 RCIM_Linear>]}}
   }
   Save cached pose 'grounded_motion_cache'(Pose=Component To Local(ComponentPose=@n28)) {cacheName=grounded_motion_cache}
   n29 = Pose Driver - Source: clavicle_l!(SourcePose:Pose Link Structure=Use cached pose 'copypose_cache'()) -> (Pose:Pose Link Structure) {
    is=title AnimGraphNode_PoseDriver
    poseAsset=&/M_MED_BlackMonday_Houston_Pack_DBNOPoses_PoseAsset
    onlyDriveBones [
     {boneName=clavicle_cape_l}
     {boneName=cape_01_l}
     {boneName=backpack_root}
     {boneName=cape_mid_01_l}
     {boneName=cape_02_l}
     {boneName=cape_mid_02_l}
     {boneName=cape_05_l}
     {boneName=cape_mid_04_l}
     {boneName=cape_04_l}
     {boneName=cape_mid_03_l}
    ]
    poseTargets [
     {
      bApplyCustomCurve=false
      bIsHidden=false
      distanceMethod=DefaultMethod
      drivenName=Pose_8
      functionType=DefaultFunction
      targetScale=1
      boneTransforms [{targetRotation=<rot -69.4483 -70.9006 78.8028>, targetTranslation=<4.3575 -0.1466 -2.3223>}]
      customCurve {defaultValue=3.4028234663852886e+38, postInfinityExtrap=RCCE_Constant, preInfinityExtrap=RCCE_Constant, keys []}
      targetRotation=<rot 0 0 0>
     }
     {
      bApplyCustomCurve=false
      bIsHidden=false
      distanceMethod=DefaultMethod
      drivenName=Pose_9
      functionType=DefaultFunction
      targetScale=0.3156
      boneTransforms [{targetRotation=<rot -50.3639 32.309 -36.3413>, targetTranslation=<4.3575 -0.1466 -2.3223>}]
      customCurve {defaultValue=3.4028234663852886e+38, postInfinityExtrap=RCCE_Constant, preInfinityExtrap=RCCE_Constant, keys []}
      targetRotation=<rot 0 0 0>
     }
     {
      bApplyCustomCurve=false
      bIsHidden=false
      distanceMethod=DefaultMethod
      drivenName=Pose_10
      functionType=DefaultFunction
      targetScale=0.8105
      boneTransforms [{targetRotation=<rot -85.0707 7.9107 -2.492>, targetTranslation=<4.3575 -0.1466 -2.3223>}]
      customCurve {defaultValue=3.4028234663852886e+38, postInfinityExtrap=RCCE_Constant, preInfinityExtrap=RCCE_Constant, keys []}
      targetRotation=<rot 0 0 0>
     }
     {
      bApplyCustomCurve=false
      bIsHidden=false
      distanceMethod=DefaultMethod
      drivenName=Pose_11
      functionType=DefaultFunction
      targetScale=0.155
      boneTransforms [{targetRotation=<rot -68.3797 28.8647 -26.7718>, targetTranslation=<4.3575 -0.1466 -2.3223>}]
      customCurve {defaultValue=3.4028234663852886e+38, postInfinityExtrap=RCCE_Constant, preInfinityExtrap=RCCE_Constant, keys []}
      targetRotation=<rot 0 0 0>
     }
     {
      bApplyCustomCurve=false
      bIsHidden=false
      distanceMethod=DefaultMethod
      drivenName=Pose_13
      functionType=DefaultFunction
      targetScale=0.155
      boneTransforms [{targetRotation=<rot -68.882 20.3963 -17.6035>, targetTranslation=<4.3575 -0.1466 -2.3223>}]
      customCurve {defaultValue=3.4028234663852886e+38, postInfinityExtrap=RCCE_Constant, preInfinityExtrap=RCCE_Constant, keys []}
      targetRotation=<rot 0 0 0>
     }
     {
      bApplyCustomCurve=false
      bIsHidden=false
      distanceMethod=DefaultMethod
      drivenName=Pose_14
      functionType=DefaultFunction
      targetScale=0.2242
      boneTransforms [{targetRotation=<rot -56.726 32.2174 -33.7614>, targetTranslation=<4.3575 -0.1466 -2.3223>}]
      customCurve {defaultValue=3.4028234663852886e+38, postInfinityExtrap=RCCE_Constant, preInfinityExtrap=RCCE_Constant, keys []}
      targetRotation=<rot 0 0 0>
     }
     {
      bApplyCustomCurve=false
      bIsHidden=false
      distanceMethod=DefaultMethod
      drivenName=Pose_15
      functionType=DefaultFunction
      targetScale=0.2242
      boneTransforms [{targetRotation=<rot -60.8617 28.6954 -28.4731>, targetTranslation=<4.3575 -0.1466 -2.3223>}]
      customCurve {defaultValue=3.4028234663852886e+38, postInfinityExtrap=RCCE_Constant, preInfinityExtrap=RCCE_Constant, keys []}
      targetRotation=<rot 0 0 0>
     }
    ]
    rBFParams {normalizeMethod=AlwaysNormalize, radius=30}
    sourceBones [{boneName=clavicle_l}]
   }
   n30 = Pose Driver - Source: clavicle_r!(SourcePose:Pose Link Structure=@n29) -> (Pose:Pose Link Structure) {
    is=title AnimGraphNode_PoseDriver
    axisLength=-20
    poseAsset=&M_MED_BlackMonday_Houston_Pack_DBNOPoses_PoseAsset
    onlyDriveBones [
     {boneName=clavicle_cape_r}
     {boneName=cape_01_r}
     {boneName=backpack_root}
     {boneName=cape_02_r}
     {boneName=cape_mid_01_r}
     {boneName=cape_05_r}
     {boneName=cape_mid_04_r}
     {boneName=cape_mid_03_l}
     {boneName=cape_04_r}
    ]
    poseTargets [
     {
      bApplyCustomCurve=false
      bIsHidden=false
      distanceMethod=DefaultMethod
      drivenName=Pose_8
      functionType=DefaultFunction
      targetScale=0.8921
      boneTransforms [{targetRotation=<rot -68.8344 103.1231 83.3014>, targetTranslation=<4.3575 -0.1466 2.3223>}]
      customCurve {defaultValue=3.4028234663852886e+38, postInfinityExtrap=RCCE_Constant, preInfinityExtrap=RCCE_Constant, keys []}
      targetRotation=<rot 0 0 0>
     }
     {
      bApplyCustomCurve=false
      bIsHidden=false
      distanceMethod=DefaultMethod
      drivenName=Pose_9
      functionType=DefaultFunction
      targetScale=0.808
      boneTransforms [{targetRotation=<rot -62.1315 174.2855 11.6426>, targetTranslation=<4.3575 -0.1466 2.3223>}]
      customCurve {defaultValue=3.4028234663852886e+38, postInfinityExtrap=RCCE_Constant, preInfinityExtrap=RCCE_Constant, keys []}
      targetRotation=<rot 0 0 0>
     }
     {
      bApplyCustomCurve=false
      bIsHidden=false
      distanceMethod=DefaultMethod
      drivenName=Pose_10
      functionType=DefaultFunction
      targetScale=0.8921
      boneTransforms [{targetRotation=<rot -58.7924 98.8576 88.1355>, targetTranslation=<4.3575 -0.1466 2.3223>}]
      customCurve {defaultValue=3.4028234663852886e+38, postInfinityExtrap=RCCE_Constant, preInfinityExtrap=RCCE_Constant, keys []}
      targetRotation=<rot 0 0 0>
     }
     {
      bApplyCustomCurve=false
      bIsHidden=false
      distanceMethod=DefaultMethod
      drivenName=Pose_11
      functionType=DefaultFunction
      targetScale=1
      boneTransforms [{targetRotation=<rot -64.2273 -133.6186 -46.2578>, targetTranslation=<4.3575 -0.1466 2.3223>}]
      customCurve {defaultValue=3.4028234663852886e+38, postInfinityExtrap=RCCE_Constant, preInfinityExtrap=RCCE_Constant, keys []}
      targetRotation=<rot 0 0 0>
     }
     {
      bApplyCustomCurve=false
      bIsHidden=false
      distanceMethod=DefaultMethod
      drivenName=Pose_14
      functionType=DefaultFunction
      targetScale=0.5214
      boneTransforms [{targetRotation=<rot -68.0727 -168.7164 -7.2706>, targetTranslation=<4.3575 -0.1466 2.3223>}]
      customCurve {defaultValue=3.4028234663852886e+38, postInfinityExtrap=RCCE_Constant, preInfinityExtrap=RCCE_Constant, keys []}
      targetRotation=<rot 0 0 0>
     }
     {
      bApplyCustomCurve=false
      bIsHidden=false
      distanceMethod=DefaultMethod
      drivenName=Pose_15
      functionType=DefaultFunction
      targetScale=0.5214
      boneTransforms [{targetRotation=<rot -63.3908 -159.6649 -17.8471>, targetTranslation=<4.3575 -0.1466 2.3223>}]
      customCurve {defaultValue=3.4028234663852886e+38, postInfinityExtrap=RCCE_Constant, preInfinityExtrap=RCCE_Constant, keys []}
      targetRotation=<rot 0 0 0>
     }
    ]
    rBFParams {normalizeMethod=AlwaysNormalize, radius=15}
    sourceBones [{boneName=clavicle_r}]
   }
   Save cached pose 'dbno_cache'(Pose=@n30) {cacheName=dbno_cache}
   AnimGraphNode_StateMachine_1 = Cape Poses() {is=title AnimGraphNode_StateMachine}
   Save cached pose 'copypose_cache'(Pose=Apply Additive(
     Base=Copy Pose From Mesh() {bUseAttachedParent=true}
     Additive=@AnimGraphNode_StateMachine_1
     Alpha=Get is_dbno_alpha() -> (is_dbno_alpha:double)
    ) {alphaBoolBlend {blendOption=Linear}}) {cacheName=copypose_cache}
   n31 = Pose Driver - Source: clavicle_l(SourcePose=Use cached pose 'copypose_cache'()) {
    poseAsset=&/M_MED_BlackMonday_Houston_Pack_Grounded_RBF_PoseAsset
    onlyDriveBones [{boneName=clavicle_cape_l}, {boneName=cape_01_l}, {boneName=backpack_root}, {boneName=cape_mid_01_l}, {boneName=cape_02_l}]
    poseTargets [
     {
      bApplyCustomCurve=false
      bIsHidden=false
      distanceMethod=DefaultMethod
      drivenName=Pose_77
      functionType=DefaultFunction
      targetScale=0.1721
      boneTransforms [{targetRotation=<rot -85.1915 -138.5625 143.0619>, targetTranslation=<4.3574 -0.1466 -2.3223>}]
      customCurve {defaultValue=3.4028234663852886e+38, postInfinityExtrap=RCCE_Constant, preInfinityExtrap=RCCE_Constant, keys []}
      targetRotation=<rot 0 0 0>
     }
     {
      bApplyCustomCurve=false
      bIsHidden=false
      distanceMethod=DefaultMethod
      drivenName=Pose_78
      functionType=DefaultFunction
      targetScale=0.4106
      boneTransforms [{targetRotation=<rot -71.8134 -5.1651 -35.1924>, targetTranslation=<4.3574 -0.1466 -2.3223>}]
      customCurve {defaultValue=3.4028234663852886e+38, postInfinityExtrap=RCCE_Constant, preInfinityExtrap=RCCE_Constant, keys []}
      targetRotation=<rot 0 0 0>
     }
     {
      bApplyCustomCurve=false
      bIsHidden=false
      distanceMethod=DefaultMethod
      drivenName=Pose_79
      functionType=DefaultFunction
      targetScale=0.3293
      boneTransforms [{targetRotation=<rot -75.5113 21.5384 -14.8289>, targetTranslation=<4.3574 -0.1466 -2.3223>}]
      customCurve {defaultValue=3.4028234663852886e+38, postInfinityExtrap=RCCE_Constant, preInfinityExtrap=RCCE_Constant, keys []}
      targetRotation=<rot 0 0 0>
     }
     {
      bApplyCustomCurve=false
      bIsHidden=false
      distanceMethod=DefaultMethod
      drivenName=Pose_80
      functionType=DefaultFunction
      targetScale=0.3454
      boneTransforms [{targetRotation=<rot -74.6776 57.008 -51.5079>, targetTranslation=<4.3574 -0.1466 -2.3223>}]
      customCurve {defaultValue=3.4028234663852886e+38, postInfinityExtrap=RCCE_Constant, preInfinityExtrap=RCCE_Constant, keys []}
      targetRotation=<rot 0 0 0>
     }
     {
      bApplyCustomCurve=false
      bIsHidden=false
      distanceMethod=DefaultMethod
      drivenName=Pose_81
      functionType=DefaultFunction
      targetScale=0.1721
      boneTransforms [{targetRotation=<rot -81.7481 -134.8649 139.6295>, targetTranslation=<4.3574 -0.1466 -2.3223>}]
      customCurve {defaultValue=3.4028234663852886e+38, postInfinityExtrap=RCCE_Constant, preInfinityExtrap=RCCE_Constant, keys []}
      targetRotation=<rot 0 0 0>
     }
     {
      bApplyCustomCurve=false
      bIsHidden=false
      distanceMethod=DefaultMethod
      drivenName=Pose_82
      functionType=DefaultFunction
      targetScale=0.1761
      boneTransforms [{targetRotation=<rot -80.8571 41.3427 -40.4103>, targetTranslation=<4.3574 -0.1466 -2.3223>}]
      customCurve {defaultValue=3.4028234663852886e+38, postInfinityExtrap=RCCE_Constant, preInfinityExtrap=RCCE_Constant, keys []}
      targetRotation=<rot 0 0 0>
     }
     {
      bApplyCustomCurve=false
      bIsHidden=false
      distanceMethod=DefaultMethod
      drivenName=Pose_83
      functionType=DefaultFunction
      targetScale=0.1564
      boneTransforms [{targetRotation=<rot -47.6434 16.3521 -2.0489>, targetTranslation=<4.3574 -0.1466 -2.3223>}]
      customCurve {defaultValue=3.4028234663852886e+38, postInfinityExtrap=RCCE_Constant, preInfinityExtrap=RCCE_Constant, keys []}
      targetRotation=<rot 0 0 0>
     }
     {
      bApplyCustomCurve=false
      bIsHidden=false
      distanceMethod=DefaultMethod
      drivenName=Pose_84
      functionType=DefaultFunction
      targetScale=0.2758
      boneTransforms [{targetRotation=<rot -85.9095 144.549 -138.9147>, targetTranslation=<4.3574 -0.1466 -2.3223>}]
      customCurve {defaultValue=3.4028234663852886e+38, postInfinityExtrap=RCCE_Constant, preInfinityExtrap=RCCE_Constant, keys []}
      targetRotation=<rot 0 0 0>
     }
     {
      bApplyCustomCurve=false
      bIsHidden=false
      distanceMethod=DefaultMethod
      drivenName=Pose_85
      functionType=DefaultFunction
      targetScale=0.1761
      boneTransforms [{targetRotation=<rot -83.9858 54.3319 -50.3537>, targetTranslation=<4.3574 -0.1466 -2.3223>}]
      customCurve {defaultValue=3.4028234663852886e+38, postInfinityExtrap=RCCE_Constant, preInfinityExtrap=RCCE_Constant, keys []}
      targetRotation=<rot 0 0 0>
     }
     {
      bApplyCustomCurve=false
      bIsHidden=false
      distanceMethod=DefaultMethod
      drivenName=Pose_86
      functionType=DefaultFunction
      targetScale=0.1564
      boneTransforms [{targetRotation=<rot -48.4262 11.785 -4.0637>, targetTranslation=<4.3574 -0.1466 -2.3223>}]
      customCurve {defaultValue=3.4028234663852886e+38, postInfinityExtrap=RCCE_Constant, preInfinityExtrap=RCCE_Constant, keys []}
      targetRotation=<rot 0 0 0>
     }
     {
      bApplyCustomCurve=false
      bIsHidden=false
      distanceMethod=DefaultMethod
      drivenName=Pose_87
      functionType=DefaultFunction
      targetScale=0.3983
      boneTransforms [{targetRotation=<rot -79.1654 101.3241 -112.608>, targetTranslation=<4.3574 -0.1466 -2.3223>}]
      customCurve {defaultValue=3.4028234663852886e+38, postInfinityExtrap=RCCE_Constant, preInfinityExtrap=RCCE_Constant, keys []}
      targetRotation=<rot 0 0 0>
     }
     {
      bApplyCustomCurve=false
      bIsHidden=false
      distanceMethod=DefaultMethod
      drivenName=Pose_88
      functionType=DefaultFunction
      targetScale=0.7012
      boneTransforms [{targetRotation=<rot -61.6146 28.9143 -24.0665>, targetTranslation=<4.3574 -0.1466 -2.3223>}]
      customCurve {defaultValue=3.4028234663852886e+38, postInfinityExtrap=RCCE_Constant, preInfinityExtrap=RCCE_Constant, keys []}
      targetRotation=<rot 0 0 0>
     }
     {
      bApplyCustomCurve=false
      bIsHidden=false
      distanceMethod=DefaultMethod
      drivenName=Pose_89
      functionType=DefaultFunction
      targetScale=0.6597
      boneTransforms [{targetRotation=<rot -58.5638 -7.8408 1.0125>, targetTranslation=<4.3574 -0.1466 -2.3223>}]
      customCurve {defaultValue=3.4028234663852886e+38, postInfinityExtrap=RCCE_Constant, preInfinityExtrap=RCCE_Constant, keys []}
      targetRotation=<rot 0 0 0>
     }
     {
      bApplyCustomCurve=false
      bIsHidden=false
      distanceMethod=DefaultMethod
      drivenName=Pose_90
      functionType=DefaultFunction
      targetScale=1
      boneTransforms [{targetRotation=<rot -29.6946 2.0811 4.5817>, targetTranslation=<4.3574 -0.1466 -2.3223>}]
      customCurve {defaultValue=3.4028234663852886e+38, postInfinityExtrap=RCCE_Constant, preInfinityExtrap=RCCE_Constant, keys []}
      targetRotation=<rot 0 0 0>
     }
    ]
    rBFParams {radius=20}
    sourceBones [{boneName=clavicle_l}]
   }
   n32 = Pose Driver - Source: clavicle_r(SourcePose=@n31) {
    axisLength=-20
    poseAsset=&M_MED_BlackMonday_Houston_Pack_Grounded_RBF_PoseAsset
    onlyDriveBones [{boneName=clavicle_cape_r}, {boneName=cape_01_r}, {boneName=backpack_root}, {boneName=cape_mid_01_r}, {boneName=cape_02_r}]
    poseTargets [
     {
      bApplyCustomCurve=false
      bIsHidden=false
      distanceMethod=DefaultMethod
      drivenName=Pose_77
      functionType=DefaultFunction
      targetScale=0.4264
      boneTransforms [{targetRotation=<rot -85.1916 41.4375 143.0619>, targetTranslation=<4.3575 -0.1466 2.3223>}]
      customCurve {defaultValue=3.4028234663852886e+38, postInfinityExtrap=RCCE_Constant, preInfinityExtrap=RCCE_Constant, keys []}
      targetRotation=<rot 0 0 0>
     }
     {
      bApplyCustomCurve=false
      bIsHidden=false
      distanceMethod=DefaultMethod
      drivenName=Pose_78
      functionType=DefaultFunction
      targetScale=0.7539
      boneTransforms [{targetRotation=<rot -64.182 -118.9543 -84.9215>, targetTranslation=<4.3575 -0.1466 2.3223>}]
      customCurve {defaultValue=3.4028234663852886e+38, postInfinityExtrap=RCCE_Constant, preInfinityExtrap=RCCE_Constant, keys []}
      targetRotation=<rot 0 0 0>
     }
     {
      bApplyCustomCurve=false
      bIsHidden=false
      distanceMethod=DefaultMethod
      drivenName=Pose_79
      functionType=DefaultFunction
      targetScale=0.8963
      boneTransforms [{targetRotation=<rot -69.946 167.2949 20.9954>, targetTranslation=<4.3575 -0.1466 2.3223>}]
      customCurve {defaultValue=3.4028234663852886e+38, postInfinityExtrap=RCCE_Constant, preInfinityExtrap=RCCE_Constant, keys []}
      targetRotation=<rot 0 0 0>
     }
     {
      bApplyCustomCurve=false
      bIsHidden=false
      distanceMethod=DefaultMethod
      drivenName=Pose_80
      functionType=DefaultFunction
      targetScale=0.5515
      boneTransforms [{targetRotation=<rot -73.4953 -175.1246 2.5328>, targetTranslation=<4.3575 -0.1466 2.3223>}]
      customCurve {defaultValue=3.4028234663852886e+38, postInfinityExtrap=RCCE_Constant, preInfinityExtrap=RCCE_Constant, keys []}
      targetRotation=<rot 0 0 0>
     }
     {
      bApplyCustomCurve=false
      bIsHidden=false
      distanceMethod=DefaultMethod
      drivenName=Pose_81
      functionType=DefaultFunction
      targetScale=0.4471
      boneTransforms [{targetRotation=<rot -87.8687 74.9818 110.3656>, targetTranslation=<4.3575 -0.1466 2.3223>}]
      customCurve {defaultValue=3.4028234663852886e+38, postInfinityExtrap=RCCE_Constant, preInfinityExtrap=RCCE_Constant, keys []}
      targetRotation=<rot 0 0 0>
     }
     {
      bApplyCustomCurve=false
      bIsHidden=false
      distanceMethod=DefaultMethod
      drivenName=Pose_82
      functionType=DefaultFunction
      targetScale=0.5515
      boneTransforms [{targetRotation=<rot -76.3657 174.0325 7.6846>, targetTranslation=<4.3575 -0.1466 2.3223>}]
      customCurve {defaultValue=3.4028234663852886e+38, postInfinityExtrap=RCCE_Constant, preInfinityExtrap=RCCE_Constant, keys []}
      targetRotation=<rot 0 0 0>
     }
     {
      bApplyCustomCurve=false
      bIsHidden=false
      distanceMethod=DefaultMethod
      drivenName=Pose_83
      functionType=DefaultFunction
      targetScale=0.836
      boneTransforms [{targetRotation=<rot -74.8444 -130.7229 -48.9381>, targetTranslation=<4.3575 -0.1466 2.3223>}]
      customCurve {defaultValue=3.4028234663852886e+38, postInfinityExtrap=RCCE_Constant, preInfinityExtrap=RCCE_Constant, keys []}
      targetRotation=<rot 0 0 0>
     }
     {
      bApplyCustomCurve=false
      bIsHidden=false
      distanceMethod=DefaultMethod
      drivenName=Pose_84
      functionType=DefaultFunction
      targetScale=0.8957
      boneTransforms [{targetRotation=<rot -57.8795 -159.8466 -20.4893>, targetTranslation=<4.3575 -0.1466 2.3223>}]
      customCurve {defaultValue=3.4028234663852886e+38, postInfinityExtrap=RCCE_Constant, preInfinityExtrap=RCCE_Constant, keys []}
      targetRotation=<rot 0 0 0>
     }
     {
      bApplyCustomCurve=false
      bIsHidden=false
      distanceMethod=DefaultMethod
      drivenName=Pose_85
      functionType=DefaultFunction
      targetScale=0.8957
      boneTransforms [{targetRotation=<rot -53.6618 -150.9982 -30.8113>, targetTranslation=<4.3575 -0.1466 2.3223>}]
      customCurve {defaultValue=3.4028234663852886e+38, postInfinityExtrap=RCCE_Constant, preInfinityExtrap=RCCE_Constant, keys []}
      targetRotation=<rot 0 0 0>
     }
     {
      bApplyCustomCurve=false
      bIsHidden=false
      distanceMethod=DefaultMethod
      drivenName=Pose_86
      functionType=DefaultFunction
      targetScale=0.7539
      boneTransforms [{targetRotation=<rot -68.9544 -125.7867 -55.7263>, targetTranslation=<4.3575 -0.1466 2.3223>}]
      customCurve {defaultValue=3.4028234663852886e+38, postInfinityExtrap=RCCE_Constant, preInfinityExtrap=RCCE_Constant, keys []}
      targetRotation=<rot 0 0 0>
     }
     {
      bApplyCustomCurve=false
      bIsHidden=false
      distanceMethod=DefaultMethod
      drivenName=Pose_87
      functionType=DefaultFunction
      targetScale=0.4264
      boneTransforms [{targetRotation=<rot -83.6816 69.9338 114.717>, targetTranslation=<4.3575 -0.1466 2.3223>}]
      customCurve {defaultValue=3.4028234663852886e+38, postInfinityExtrap=RCCE_Constant, preInfinityExtrap=RCCE_Constant, keys []}
      targetRotation=<rot 0 0 0>
     }
     {
      bApplyCustomCurve=false
      bIsHidden=false
      distanceMethod=DefaultMethod
      drivenName=Pose_88
      functionType=DefaultFunction
      targetScale=0.9049
      boneTransforms [{targetRotation=<rot -66.9199 -176.0527 1.2117>, targetTranslation=<4.3575 -0.1466 2.3223>}]
      customCurve {defaultValue=3.4028234663852886e+38, postInfinityExtrap=RCCE_Constant, preInfinityExtrap=RCCE_Constant, keys []}
      targetRotation=<rot 0 0 0>
     }
     {
      bApplyCustomCurve=false
      bIsHidden=false
      distanceMethod=DefaultMethod
      drivenName=Pose_89
      functionType=DefaultFunction
      targetScale=1
      boneTransforms [{targetRotation=<rot -47.9631 -143.824 -33.4383>, targetTranslation=<4.3575 -0.1466 2.3223>}]
      customCurve {defaultValue=3.4028234663852886e+38, postInfinityExtrap=RCCE_Constant, preInfinityExtrap=RCCE_Constant, keys []}
      targetRotation=<rot 0 0 0>
     }
     {
      bApplyCustomCurve=false
      bIsHidden=false
      distanceMethod=DefaultMethod
      drivenName=Pose_90
      functionType=DefaultFunction
      targetScale=0.6925
      boneTransforms [{targetRotation=<rot -86.6962 -150.2418 -25.3507>, targetTranslation=<4.3575 -0.1466 2.3223>}]
      customCurve {defaultValue=3.4028234663852886e+38, postInfinityExtrap=RCCE_Constant, preInfinityExtrap=RCCE_Constant, keys []}
      targetRotation=<rot 0 0 0>
     }
    ]
    rBFParams {radius=10}
    sourceBones [{boneName=clavicle_r}]
   }
   Save cached pose 'grounded_pre_sim_cache'(Pose=@n32) {cacheName=grounded_pre_sim_cache}
   n33 = Pose Driver - Source: clavicle_l(SourcePose=Use cached pose 'copypose_cache'()) {
    poseAsset=&/M_MED_BlackMonday_Houston_Pack_Skydive_RBF_PoseAsset
    onlyDriveBones [{boneName=clavicle_cape_l}, {boneName=cape_01_l}, {boneName=backpack_root}, {boneName=cape_mid_01_l}, {boneName=cape_02_l}]
    poseTargets [
     {
      bApplyCustomCurve=false
      bIsHidden=false
      distanceMethod=DefaultMethod
      drivenName=Pose_16
      functionType=DefaultFunction
      targetScale=1
      boneTransforms [{targetRotation=<rot -85.1915 -138.5625 143.0619>, targetTranslation=<4.3574 -0.1466 -2.3223>}]
      customCurve {defaultValue=3.4028234663852886e+38, postInfinityExtrap=RCCE_Constant, preInfinityExtrap=RCCE_Constant, keys []}
      targetRotation=<rot 0 0 0>
     }
     {
      bApplyCustomCurve=false
      bIsHidden=false
      distanceMethod=DefaultMethod
      drivenName=Pose_17
      functionType=DefaultFunction
      targetScale=0.6922
      boneTransforms [{targetRotation=<rot -56.9158 -106.949 109.3363>, targetTranslation=<4.3574 -0.1466 -2.3223>}]
      customCurve {defaultValue=3.4028234663852886e+38, postInfinityExtrap=RCCE_Constant, preInfinityExtrap=RCCE_Constant, keys []}
      targetRotation=<rot 0 0 0>
     }
     {
      bApplyCustomCurve=false
      bIsHidden=false
      distanceMethod=DefaultMethod
      drivenName=Pose_18
      functionType=DefaultFunction
      targetScale=0.167
      boneTransforms [{targetRotation=<rot -62.8696 10.5972 -4.8714>, targetTranslation=<4.3574 -0.1466 -2.3223>}]
      customCurve {defaultValue=3.4028234663852886e+38, postInfinityExtrap=RCCE_Constant, preInfinityExtrap=RCCE_Constant, keys []}
      targetRotation=<rot 0 0 0>
     }
     {
      bApplyCustomCurve=false
      bIsHidden=false
      distanceMethod=DefaultMethod
      drivenName=Pose_19
      functionType=DefaultFunction
      targetScale=0.1821
      boneTransforms [{targetRotation=<rot -58.134 -81.628 86.7679>, targetTranslation=<4.3574 -0.1466 -2.3223>}]
      customCurve {defaultValue=3.4028234663852886e+38, postInfinityExtrap=RCCE_Constant, preInfinityExtrap=RCCE_Constant, keys []}
      targetRotation=<rot 0 0 0>
     }
     {
      bApplyCustomCurve=false
      bIsHidden=false
      distanceMethod=DefaultMethod
      drivenName=Pose_20
      functionType=DefaultFunction
      targetScale=0.1821
      boneTransforms [{targetRotation=<rot -56.6824 -86.3988 91.2274>, targetTranslation=<4.3574 -0.1466 -2.3223>}]
      customCurve {defaultValue=3.4028234663852886e+38, postInfinityExtrap=RCCE_Constant, preInfinityExtrap=RCCE_Constant, keys []}
      targetRotation=<rot 0 0 0>
     }
     {
      bApplyCustomCurve=false
      bIsHidden=false
      distanceMethod=DefaultMethod
      drivenName=Pose_21
      functionType=DefaultFunction
      targetScale=0.1973
      boneTransforms [{targetRotation=<rot -62.9578 22.6411 -17.858>, targetTranslation=<4.3574 -0.1466 -2.3223>}]
      customCurve {defaultValue=3.4028234663852886e+38, postInfinityExtrap=RCCE_Constant, preInfinityExtrap=RCCE_Constant, keys []}
      targetRotation=<rot 0 0 0>
     }
     {
      bApplyCustomCurve=false
      bIsHidden=false
      distanceMethod=DefaultMethod
      drivenName=Pose_22
      functionType=DefaultFunction
      targetScale=0.167
      boneTransforms [{targetRotation=<rot -61.9412 16.085 -10.5734>, targetTranslation=<4.3574 -0.1466 -2.3223>}]
      customCurve {defaultValue=3.4028234663852886e+38, postInfinityExtrap=RCCE_Constant, preInfinityExtrap=RCCE_Constant, keys []}
      targetRotation=<rot 0 0 0>
     }
     {
      bApplyCustomCurve=false
      bIsHidden=false
      distanceMethod=DefaultMethod
      drivenName=Pose_23
      functionType=DefaultFunction
      targetScale=1
      boneTransforms [{targetRotation=<rot -77.4669 88.3019 -84.1615>, targetTranslation=<4.3574 -0.1466 -2.3223>}]
      customCurve {defaultValue=3.4028234663852886e+38, postInfinityExtrap=RCCE_Constant, preInfinityExtrap=RCCE_Constant, keys []}
      targetRotation=<rot 0 0 0>
     }
    ]
    rBFParams {normalizeMethod=AlwaysNormalize, radius=16}
    sourceBones [{boneName=clavicle_l}]
   }
   n34 = Pose Driver - Source: clavicle_r(SourcePose=@n33) {
    axisLength=-20
    poseAsset=&M_MED_BlackMonday_Houston_Pack_Skydive_RBF_PoseAsset
    onlyDriveBones [{boneName=clavicle_cape_r}, {boneName=cape_01_r}, {boneName=backpack_root}, {boneName=cape_mid_01_r}, {boneName=cape_02_r}]
    poseTargets [
     {
      bApplyCustomCurve=false
      bIsHidden=false
      distanceMethod=DefaultMethod
      drivenName=Pose_16
      functionType=DefaultFunction
      targetScale=0.8776
      boneTransforms [{targetRotation=<rot -85.1916 41.4375 143.0619>, targetTranslation=<4.3575 -0.1466 2.3223>}]
      customCurve {defaultValue=3.4028234663852886e+38, postInfinityExtrap=RCCE_Constant, preInfinityExtrap=RCCE_Constant, keys []}
      targetRotation=<rot 0 0 0>
     }
     {
      bApplyCustomCurve=false
      bIsHidden=false
      distanceMethod=DefaultMethod
      drivenName=Pose_17
      functionType=DefaultFunction
      targetScale=0.4631
      boneTransforms [{targetRotation=<rot -56.2815 96.6907 90.6957>, targetTranslation=<4.3575 -0.1466 2.3223>}]
      customCurve {defaultValue=3.4028234663852886e+38, postInfinityExtrap=RCCE_Constant, preInfinityExtrap=RCCE_Constant, keys []}
      targetRotation=<rot 0 0 0>
     }
     {
      bApplyCustomCurve=false
      bIsHidden=false
      distanceMethod=DefaultMethod
      drivenName=Pose_18
      functionType=DefaultFunction
      targetScale=0.4631
      boneTransforms [{targetRotation=<rot -57.8513 76.9348 106.696>, targetTranslation=<4.3575 -0.1466 2.3223>}]
      customCurve {defaultValue=3.4028234663852886e+38, postInfinityExtrap=RCCE_Constant, preInfinityExtrap=RCCE_Constant, keys []}
      targetRotation=<rot 0 0 0>
     }
     {
      bApplyCustomCurve=false
      bIsHidden=false
      distanceMethod=DefaultMethod
      drivenName=Pose_19
      functionType=DefaultFunction
      targetScale=0.8776
      boneTransforms [{targetRotation=<rot -65.1335 19.0932 158.6613>, targetTranslation=<4.3575 -0.1466 2.3223>}]
      customCurve {defaultValue=3.4028234663852886e+38, postInfinityExtrap=RCCE_Constant, preInfinityExtrap=RCCE_Constant, keys []}
      targetRotation=<rot 0 0 0>
     }
     {
      bApplyCustomCurve=false
      bIsHidden=false
      distanceMethod=DefaultMethod
      drivenName=Pose_20
      functionType=DefaultFunction
      targetScale=1
      boneTransforms [{targetRotation=<rot -70.681 -102.0605 -78.3646>, targetTranslation=<4.3575 -0.1466 2.3223>}]
      customCurve {defaultValue=3.4028234663852886e+38, postInfinityExtrap=RCCE_Constant, preInfinityExtrap=RCCE_Constant, keys []}
      targetRotation=<rot 0 0 0>
     }
     {
      bApplyCustomCurve=false
      bIsHidden=false
      distanceMethod=DefaultMethod
      drivenName=Pose_21
      functionType=DefaultFunction
      targetScale=0.6958
      boneTransforms [{targetRotation=<rot -61.9506 126.7251 60.562>, targetTranslation=<4.3575 -0.1466 2.3223>}]
      customCurve {defaultValue=3.4028234663852886e+38, postInfinityExtrap=RCCE_Constant, preInfinityExtrap=RCCE_Constant, keys []}
      targetRotation=<rot 0 0 0>
     }
     {
      bApplyCustomCurve=false
      bIsHidden=false
      distanceMethod=DefaultMethod
      drivenName=Pose_22
      functionType=DefaultFunction
      targetScale=0.18
      boneTransforms [{targetRotation=<rot -59.8998 -165.3406 -9.2548>, targetTranslation=<4.3575 -0.1466 2.3223>}]
      customCurve {defaultValue=3.4028234663852886e+38, postInfinityExtrap=RCCE_Constant, preInfinityExtrap=RCCE_Constant, keys []}
      targetRotation=<rot 0 0 0>
     }
     {
      bApplyCustomCurve=false
      bIsHidden=false
      distanceMethod=DefaultMethod
      drivenName=Pose_23
      functionType=DefaultFunction
      targetScale=0.18
      boneTransforms [{targetRotation=<rot -59.8898 -156.9567 -74.4791>, targetTranslation=<4.3575 -0.1466 2.3223>}]
      customCurve {defaultValue=3.4028234663852886e+38, postInfinityExtrap=RCCE_Constant, preInfinityExtrap=RCCE_Constant, keys []}
      targetRotation=<rot 0 0 0>
     }
    ]
    rBFParams {normalizeMethod=AlwaysNormalize, radius=20}
    sourceBones [{boneName=clavicle_r}]
   }
   Save cached pose 'skydive_pre_sim_cache'(Pose=@n34) {cacheName=skydive_pre_sim_cache}
  ]}
 Cape Poses {
  entry=Base
  path=AnimGraph.AnimGraphNode_StateMachine_1.Cape Poses
  ids=id
  nodes [
   AnimStateNode_0 = Base() {is=title AnimStateNode}
   AnimStateNode_1 = Jumping() {is=title AnimStateNode}
   AnimStateNode_2 = Falling() {is=title AnimStateNode}
   AnimStateNode_3 = Targeting() {is=title AnimStateNode}
   AnimStateNode_5 = SkydiveMode() {is=title AnimStateNode}
   AnimStateNode_9 = Crouching() {is=title AnimStateNode}
   AnimStateTransitionNode_0 = AnimStateTransitionNode() {blendMode=Custom, customBlendCurve=/Game/Accessories/FORT_Backpacks/M_MED_Demon_Pack/Animation/JumpTransitionBlendCurve, priorityOrder=2}
   AnimStateTransitionNode_1 = AnimStateTransitionNode() {
    blendMode=Custom
    crossfadeDuration=0.5
    customBlendCurve=/Game/Accessories/FORT_Backpacks/M_MED_Demon_Pack/Animation/JumpTransitionBlendOutCurve
   }
   AnimStateTransitionNode_11 = AnimStateTransitionNode()
   AnimStateTransitionNode_12 = AnimStateTransitionNode()
   AnimStateTransitionNode_13 = AnimStateTransitionNode()
   AnimStateTransitionNode_15 = AnimStateTransitionNode()
   AnimStateTransitionNode_2 = AnimStateTransitionNode()
   AnimStateTransitionNode_3 = AnimStateTransitionNode() {blendMode=Custom, customBlendCurve=&JumpTransitionBlendOutCurve}
   AnimStateTransitionNode_4 = AnimStateTransitionNode() {blendMode=Custom, customBlendCurve=&JumpTransitionBlendCurve}
   AnimStateTransitionNode_5 = AnimStateTransitionNode()
   AnimStateTransitionNode_6 = AnimStateTransitionNode() {crossfadeDuration=0.25}
   AnimStateTransitionNode_7 = AnimStateTransitionNode() {crossfadeDuration=0.25}
  ]
  links [
   {
    dst=Crouching
    rule=AnimGraph.AnimGraphNode_StateMachine_1.Cape Poses.AnimStateTransitionNode_12.AnimationTransitionGraph_0
    src=Base
    via=AnimStateTransitionNode_12
   }
   {
    dst=Falling
    rule=AnimGraph.AnimGraphNode_StateMachine_1.Cape Poses.AnimStateTransitionNode_4.AnimationTransitionGraph_0
    src=Base
    via=AnimStateTransitionNode_4
   }
   {
    dst=Jumping
    rule=AnimGraph.AnimGraphNode_StateMachine_1.Cape Poses.AnimStateTransitionNode_0.AnimationTransitionGraph_0
    src=Base
    via=AnimStateTransitionNode_0
   }
   {
    dst=SkydiveMode
    rule=AnimGraph.AnimGraphNode_StateMachine_1.Cape Poses.AnimStateTransitionNode_15.AnimationTransitionGraph_0
    src=Base
    via=AnimStateTransitionNode_15
   }
   {
    dst=Targeting
    rule=AnimGraph.AnimGraphNode_StateMachine_1.Cape Poses.AnimStateTransitionNode_6.AnimationTransitionGraph_0
    src=Base
    via=AnimStateTransitionNode_6
   }
   {
    dst=Base
    rule=AnimGraph.AnimGraphNode_StateMachine_1.Cape Poses.AnimStateTransitionNode_13.AnimationTransitionGraph_0
    src=Crouching
    via=AnimStateTransitionNode_13
   }
   {
    dst=Base
    rule=AnimGraph.AnimGraphNode_StateMachine_1.Cape Poses.AnimStateTransitionNode_3.AnimationTransitionGraph_0
    src=Falling
    via=AnimStateTransitionNode_3
   }
   {
    dst=Base
    rule=AnimGraph.AnimGraphNode_StateMachine_1.Cape Poses.AnimStateTransitionNode_1.AnimationTransitionGraph_0
    src=Jumping
    via=AnimStateTransitionNode_1
   }
   {
    dst=Falling
    rule=AnimGraph.AnimGraphNode_StateMachine_1.Cape Poses.AnimStateTransitionNode_2.AnimationTransitionGraph_0
    src=Jumping
    via=AnimStateTransitionNode_2
   }
   {
    dst=SkydiveMode
    rule=AnimGraph.AnimGraphNode_StateMachine_1.Cape Poses.AnimStateTransitionNode_5.AnimationTransitionGraph_0
    src=Jumping
    via=AnimStateTransitionNode_5
   }
   {
    dst=Base
    rule=AnimGraph.AnimGraphNode_StateMachine_1.Cape Poses.AnimStateTransitionNode_11.AnimationTransitionGraph_0
    src=SkydiveMode
    via=AnimStateTransitionNode_11
   }
   {
    dst=Base
    rule=AnimGraph.AnimGraphNode_StateMachine_1.Cape Poses.AnimStateTransitionNode_7.AnimationTransitionGraph_0
    src=Targeting
    via=AnimStateTransitionNode_7
   }
  ]
 }
 Base {
  path=AnimGraph.AnimGraphNode_StateMachine_1.Cape Poses.AnimStateNode_0.Base
  nodes [
   n1 = Get PlayerVector_FwdBwd() -> (PlayerVector_FwdBwd:double)
   n2 = Get PlayerVector_LR() -> (PlayerVector_LR:double)
   n3 = BlackMonday_Houston_Poses_PoseAsset : Base!(PoseWeight:float=1) -> (Pose:Pose Link Structure) {
    is=title AnimGraphNode_PoseByName
    poseAsset=&BlackMonday_Houston_Poses_PoseAsset
    poseName=Base
   }
   Output Animation Pose(Result=Apply Additive(
     Base=@n3
     Additive=Blendspace Player 'BS_Houston_Leans_Ground'(X=@n2, Y=@n1) {blendSpace=/Figment_S10/Accessories/FORT_Capes/M_MED_BlackMonday_Houston_Cape/Animation/BS_Houston_Leans_Ground}
    ) {alphaBoolBlend {blendOption=Linear}})
  ]
 }
 Jumping {
  path=AnimGraph.AnimGraphNode_StateMachine_1.Cape Poses.AnimStateNode_1.Jumping
  nodes [
   n1 = Get PlayerVector_FwdBwd()
   n2 = Get PlayerVector_LR()
   n3 = BlackMonday_Houston_Poses_PoseAsset : Jump!(PoseWeight:float=1) -> (Pose:Pose Link Structure) {
    is=title AnimGraphNode_PoseByName
    poseAsset=&BlackMonday_Houston_Poses_PoseAsset
    poseName=Jump
   }
   Output Animation Pose(Result=Apply Additive(
     Base=@n3
     Additive=Blendspace Player 'BS_Houston_Leans_Ground'(X=@n2, Y=@n1) {blendSpace=&BS_Houston_Leans_Ground}
    ) {alphaBoolBlend {blendOption=Linear}})
  ]
 }
 Falling {
  path=AnimGraph.AnimGraphNode_StateMachine_1.Cape Poses.AnimStateNode_2.Falling
  nodes [
   n1 = Get PlayerVector_FwdBwd()
   n2 = Get PlayerVector_LR()
   n3 = BlackMonday_Houston_Poses_PoseAsset : Jump(PoseWeight=1) {poseAsset=&BlackMonday_Houston_Poses_PoseAsset, poseName=Jump}
   Output Animation Pose(Result=Apply Additive(
     Base=@n3
     Additive=Blendspace Player 'BS_Houston_Leans_Ground'(X=@n2, Y=@n1) {blendSpace=&BS_Houston_Leans_Ground}
    ) {alphaBoolBlend {blendOption=Linear}})
  ]
 }
 Targeting {
  path=AnimGraph.AnimGraphNode_StateMachine_1.Cape Poses.AnimStateNode_3.Targeting
  nodes [
   n1 = BlackMonday_Houston_Poses_PoseAsset : Targeting!(PoseWeight:float=0) -> (Pose:Pose Link Structure) {
    is=title AnimGraphNode_PoseByName
    poseAsset=&BlackMonday_Houston_Poses_PoseAsset
    poseName=Targeting
   }
   Output Animation Pose(Result=@n1)
  ]
 }
 SkydiveMode {
  path=AnimGraph.AnimGraphNode_StateMachine_1.Cape Poses.AnimStateNode_5.SkydiveMode
  nodes [
   BlackMonday_Houston_Poses_PoseAsset : Float!(PoseWeight:float=0) -> (Pose:Pose Link Structure) {
    is=title AnimGraphNode_PoseByName
    poseAsset=&BlackMonday_Houston_Poses_PoseAsset
    poseName=Float
   }
   AnimGraphNode_StateMachine_0 = Skydiving State() {is=title AnimGraphNode_StateMachine}
   Output Animation Pose(Result=@AnimGraphNode_StateMachine_0)
  ]
 }
 Skydiving State {
  entry=Parachuting
  path=AnimGraph.AnimGraphNode_StateMachine_1.Cape Poses.AnimStateNode_5.SkydiveMode.AnimGraphNode_StateMachine_0.Skydiving State
  ids=id
  nodes [
   AnimStateNode_1 = Diving() {is=title AnimStateNode}
   AnimStateNode_2 = Floating() {is=title AnimStateNode}
   AnimStateNode_4 = Parachuting() {is=title AnimStateNode}
   AnimStateTransitionNode_1 = AnimStateTransitionNode() {crossfadeDuration=0.5}
   AnimStateTransitionNode_10 = AnimStateTransitionNode()
   AnimStateTransitionNode_12 = AnimStateTransitionNode()
   AnimStateTransitionNode_13 = AnimStateTransitionNode()
   AnimStateTransitionNode_3 = AnimStateTransitionNode()
   AnimStateTransitionNode_5 = AnimStateTransitionNode() {crossfadeDuration=0.5}
  ]
  links [
   {
    dst=Floating
    rule=AnimGraph.AnimGraphNode_StateMachine_1.Cape Poses.AnimStateNode_5.SkydiveMode.AnimGraphNode_StateMachine_0.Skydiving State.AnimStateTransitionNode_1.AnimationTransitionGraph_0
    src=Diving
    via=AnimStateTransitionNode_1
   }
   {
    dst=Parachuting
    rule=AnimGraph.AnimGraphNode_StateMachine_1.Cape Poses.AnimStateNode_5.SkydiveMode.AnimGraphNode_StateMachine_0.Skydiving State.AnimStateTransitionNode_10.AnimationTransitionGraph_1
    src=Diving
    via=AnimStateTransitionNode_10
   }
   {
    dst=Diving
    rule=AnimGraph.AnimGraphNode_StateMachine_1.Cape Poses.AnimStateNode_5.SkydiveMode.AnimGraphNode_StateMachine_0.Skydiving State.AnimStateTransitionNode_5.AnimationTransitionGraph_0
    src=Floating
    via=AnimStateTransitionNode_5
   }
   {
    dst=Parachuting
    rule=AnimGraph.AnimGraphNode_StateMachine_1.Cape Poses.AnimStateNode_5.SkydiveMode.AnimGraphNode_StateMachine_0.Skydiving State.AnimStateTransitionNode_3.AnimationTransitionGraph_1
    src=Floating
    via=AnimStateTransitionNode_3
   }
   {
    dst=Diving
    rule=AnimGraph.AnimGraphNode_StateMachine_1.Cape Poses.AnimStateNode_5.SkydiveMode.AnimGraphNode_StateMachine_0.Skydiving State.AnimStateTransitionNode_13.AnimationTransitionGraph_1
    src=Parachuting
    via=AnimStateTransitionNode_13
   }
   {
    dst=Floating
    rule=AnimGraph.AnimGraphNode_StateMachine_1.Cape Poses.AnimStateNode_5.SkydiveMode.AnimGraphNode_StateMachine_0.Skydiving State.AnimStateTransitionNode_12.AnimationTransitionGraph_1
    src=Parachuting
    via=AnimStateTransitionNode_12
   }
  ]
 }
 Diving {
  path=AnimGraph.AnimGraphNode_StateMachine_1.Cape Poses.AnimStateNode_5.SkydiveMode.AnimGraphNode_StateMachine_0.Skydiving State.AnimStateNode_1.Diving
  nodes [
   n1 = Get PlayerVector_FwdBwd()
   n2 = Get PlayerVector_LR()
   n3 = BlackMonday_Houston_Poses_PoseAsset : Dive!(PoseWeight:float=0.1) -> (Pose:Pose Link Structure) {
    is=title AnimGraphNode_PoseByName
    poseAsset=&BlackMonday_Houston_Poses_PoseAsset
    poseName=Dive
   }
   Output Animation Pose(Result=Apply Additive(
     Base=@n3
     Additive=Blendspace Player 'BS_Houston_Leans_Parachuting'(X=@n2, Y=@n1) {blendSpace=/Figment_S10/Accessories/FORT_Capes/M_MED_BlackMonday_Houston_Cape/Animation/BS_Houston_Leans_Parachuting}
    ) {alphaBoolBlend {blendOption=Linear}})
  ]
 }
 Floating {
  path=AnimGraph.AnimGraphNode_StateMachine_1.Cape Poses.AnimStateNode_5.SkydiveMode.AnimGraphNode_StateMachine_0.Skydiving State.AnimStateNode_2.Floating
  nodes [
   n1 = Get PlayerVector_FwdBwd()
   n2 = Get PlayerVector_LR()
   n3 = BlackMonday_Houston_Poses_PoseAsset : Float(PoseWeight=0.4) {poseAsset=&BlackMonday_Houston_Poses_PoseAsset, poseName=Float}
   Output Animation Pose(Result=Apply Additive(
     Base=@n3
     Additive=Blendspace Player 'BS_Houston_Leans_Parachuting'(X=@n2, Y=@n1) {blendSpace=&BS_Houston_Leans_Parachuting}
    ) {alphaBoolBlend {blendOption=Linear}})
  ]
 }
 Parachuting {
  path=AnimGraph.AnimGraphNode_StateMachine_1.Cape Poses.AnimStateNode_5.SkydiveMode.AnimGraphNode_StateMachine_0.Skydiving State.AnimStateNode_4.Parachuting
  nodes [
   Apply Additive(Additive=Blendspace Player 'BS_Houston_Leans_Parachuting'(X=Get PlayerVector_LR(), Y=Get PlayerVector_FwdBwd()) {blendSpace=&BS_Houston_Leans_Parachuting}) {alphaBoolBlend {blendOption=Linear}}
   n1 = BlackMonday_Houston_Poses_PoseAsset : Float(PoseWeight=0.1) {poseAsset=&BlackMonday_Houston_Poses_PoseAsset, poseName=Float}
   Output Animation Pose(Result=@n1)
  ]
 }
 AnimationTransitionGraph_0 {
  path=AnimGraph.AnimGraphNode_StateMachine_1.Cape Poses.AnimStateNode_5.SkydiveMode.AnimGraphNode_StateMachine_0.Skydiving State.AnimStateTransitionNode_1.AnimationTransitionGraph_0
  nodes [Result(bCanEnterTransition=NOT Boolean(A=Get bIsSkydiveDiveMode()))]
 }
 AnimationTransitionGraph_1 {
  path=AnimGraph.AnimGraphNode_StateMachine_1.Cape Poses.AnimStateNode_5.SkydiveMode.AnimGraphNode_StateMachine_0.Skydiving State.AnimStateTransitionNode_10.AnimationTransitionGraph_1
  nodes [Result(bCanEnterTransition=Get bIsParachuteOpened())]
 }
 AnimationTransitionGraph_1 {
  path=AnimGraph.AnimGraphNode_StateMachine_1.Cape Poses.AnimStateNode_5.SkydiveMode.AnimGraphNode_StateMachine_0.Skydiving State.AnimStateTransitionNode_12.AnimationTransitionGraph_1
  nodes [Result(bCanEnterTransition=NOT Boolean(A=Get bIsParachuteOpened()))]
 }
 AnimationTransitionGraph_1 {
  path=AnimGraph.AnimGraphNode_StateMachine_1.Cape Poses.AnimStateNode_5.SkydiveMode.AnimGraphNode_StateMachine_0.Skydiving State.AnimStateTransitionNode_13.AnimationTransitionGraph_1
  nodes [Result(bCanEnterTransition=NOT Boolean(A=Get bIsParachuteOpened()))]
 }
 AnimationTransitionGraph_1 {
  path=AnimGraph.AnimGraphNode_StateMachine_1.Cape Poses.AnimStateNode_5.SkydiveMode.AnimGraphNode_StateMachine_0.Skydiving State.AnimStateTransitionNode_3.AnimationTransitionGraph_1
  nodes [Result(bCanEnterTransition=Get bIsParachuteOpened())]
 }
 AnimationTransitionGraph_0 {
  path=AnimGraph.AnimGraphNode_StateMachine_1.Cape Poses.AnimStateNode_5.SkydiveMode.AnimGraphNode_StateMachine_0.Skydiving State.AnimStateTransitionNode_5.AnimationTransitionGraph_0
  nodes [Result(bCanEnterTransition=Get bIsSkydiveDiveMode())]
 }
 Crouching {
  path=AnimGraph.AnimGraphNode_StateMachine_1.Cape Poses.AnimStateNode_9.Crouching
  nodes [
   n1 = BlackMonday_Houston_Poses_PoseAsset : Crouch!(PoseWeight:float=1) -> (Pose:Pose Link Structure) {
    is=title AnimGraphNode_PoseByName
    poseAsset=&BlackMonday_Houston_Poses_PoseAsset
    poseName=Crouch
   }
   Output Animation Pose(Result=@n1)
  ]
 }
 AnimationTransitionGraph_0 {
  path=AnimGraph.AnimGraphNode_StateMachine_1.Cape Poses.AnimStateTransitionNode_0.AnimationTransitionGraph_0
  nodes [Result(bCanEnterTransition=Get bIsJumping())]
 }
 AnimationTransitionGraph_0 {
  path=AnimGraph.AnimGraphNode_StateMachine_1.Cape Poses.AnimStateTransitionNode_1.AnimationTransitionGraph_0
  nodes [
   Result(bCanEnterTransition=NOT Boolean(A=Get bIsJumping()))
   Get bShouldPredictLanding() -> (bShouldPredictLanding:bool)
   Get bLandingPredicted() -> (bLandingPredicted:bool)
   Get bDebugLandPrediction() -> (bDebugLandPrediction:bool)
   Get bIsFalling()
  ]
 }
 AnimationTransitionGraph_0 {
  path=AnimGraph.AnimGraphNode_StateMachine_1.Cape Poses.AnimStateTransitionNode_11.AnimationTransitionGraph_0
  nodes [Result(bCanEnterTransition=NOT Boolean(A=Get bIsSkyDiving()))]
 }
 AnimationTransitionGraph_0 {
  path=AnimGraph.AnimGraphNode_StateMachine_1.Cape Poses.AnimStateTransitionNode_12.AnimationTransitionGraph_0
  nodes [Result(bCanEnterTransition=Get bIsCrouching())]
 }
 AnimationTransitionGraph_0 {
  path=AnimGraph.AnimGraphNode_StateMachine_1.Cape Poses.AnimStateTransitionNode_13.AnimationTransitionGraph_0
  nodes [Result(bCanEnterTransition=NOT Boolean(A=Get bIsCrouching()))]
 }
 AnimationTransitionGraph_0 {
  path=AnimGraph.AnimGraphNode_StateMachine_1.Cape Poses.AnimStateTransitionNode_15.AnimationTransitionGraph_0
  nodes [Result(bCanEnterTransition=Get bIsSkyDiving())]
 }
 AnimationTransitionGraph_0 {
  path=AnimGraph.AnimGraphNode_StateMachine_1.Cape Poses.AnimStateTransitionNode_2.AnimationTransitionGraph_0
  nodes [Result(bCanEnterTransition=Get bIsFalling())]
 }
 AnimationTransitionGraph_0 {
  path=AnimGraph.AnimGraphNode_StateMachine_1.Cape Poses.AnimStateTransitionNode_3.AnimationTransitionGraph_0
  nodes [Result(bCanEnterTransition=NOT Boolean(A=Get bIsFalling()))]
 }
 AnimationTransitionGraph_0 {
  path=AnimGraph.AnimGraphNode_StateMachine_1.Cape Poses.AnimStateTransitionNode_4.AnimationTransitionGraph_0
  nodes [Result(bCanEnterTransition=Get bIsFalling())]
 }
 AnimationTransitionGraph_0 {
  path=AnimGraph.AnimGraphNode_StateMachine_1.Cape Poses.AnimStateTransitionNode_5.AnimationTransitionGraph_0
  nodes [Result(bCanEnterTransition=Get bIsSkyDiving())]
 }
 AnimationTransitionGraph_0 {
  path=AnimGraph.AnimGraphNode_StateMachine_1.Cape Poses.AnimStateTransitionNode_6.AnimationTransitionGraph_0
  nodes [
   Result(bCanEnterTransition=Get bIsTargeting())
   Get is Targeting() -> (is Targeting:bool)
  ]
 }
 AnimationTransitionGraph_0 {
  path=AnimGraph.AnimGraphNode_StateMachine_1.Cape Poses.AnimStateTransitionNode_7.AnimationTransitionGraph_0
  nodes [Result(bCanEnterTransition=NOT Boolean(A=Get bIsTargeting()))]
 }
 EventGraph {nodes [
   n1 = Append(A="Current LOD: ", B="To String (Integer)"(InInt=Get CurrentLOD()))
   PrintString!(execute, InString=@n1, bPrintToScreen, bPrintToLog, TextColor=(R=1,G=0,B=0.0364,A=1), Duration=0.1, Key) -> (then)
   n2 = GetSocketLocation!(self:Scene Component Object Reference=GetPartSkeletalMeshComponent(), InSocketName:Name=head) -> (ReturnValue:Vector)
   > DrawDebugString!(
    execute
    TextLocation:Vector=@n2
    Text:String=@n1
    TestBaseActor:Actor Object Reference
    TextColor:Linear Color Structure=(R=0,G=1,B=0.0903,A=1)
    Duration:float=0
   ) -> (then)
   n3 = TryGetPawnOwner()
   > DrawDebugArrow!(
    execute
    LineStart:Vector=@n2
    LineEnd:Vector=vector * vector(A:Vector=GetActorForwardVector!(self:Actor Object Reference=@n3) -> (ReturnValue:Vector), B:Vector=@n2) -> (ReturnValue:Vector) {is=title K2Node_CommutativeAssociativeBinaryOperator}
    ArrowSize:float=3
    LineColor:Linear Color Structure=(R=0.187,G=1,B=0,A=1)
    Duration:float=0
    Thickness:float=1
    DepthPriority:EDrawDebugSceneDepthPriorityGroup Enum=World
   ) -> (then)
   n4 = GetSocketLocation(self=GetPartSkeletalMeshComponent(), InSocketName=root)
   n5 = Rotate Vector!(
    A:Vector="0,-90, 0"
    B:Rotator=Get Actor Rotation!(self:Actor Object Reference=Get player_pawn() -> (player_pawn:Fort Player Pawn Athena Object Reference)) -> (ReturnValue:Rotator) {is=title K2Node_CallFunction}
   ) -> (ReturnValue:Vector)
   > DrawDebugArrow(
    LineStart=@n4
    LineEnd=vector + vector(A:Vector=@n4, B:Vector=vector * float(A=@n5, B=2)) -> (ReturnValue:Vector) {is=title K2Node_CommutativeAssociativeBinaryOperator}
    ArrowSize=0
    LineColor=(R=0.9462,G=0,B=1,A=1)
    Duration=0
    Thickness=2
    DepthPriority=World
   )
   Event BlueprintUpdateAnimation()
   > Is Valid(InputObject=Get player_pawn())
   > n6 = Sequence!(execute) -> (then_0, then_1, then_2:Exec, then_3:Exec, then_4:Exec, then_5:Exec, then_6:Exec, then_7:Exec, then_8:Exec)
   > Do Once()
   > Set body_mesh(body_mesh:Skeletal Mesh Component Object Reference=GetSkeletalMeshForPartType(self=Get player_pawn())) -> (Output_Get:Skeletal Mesh Component Object Reference)
   > Set face_mesh(face_mesh:Skeletal Mesh Component Object Reference=GetSkeletalMeshForPartType(self=Get player_pawn(), PartType=Head)) -> (Output_Get:Skeletal Mesh Component Object Reference)
   n7 = Get player_pawn()
   n8 = GetCurrentAcceleration!(self:Character Movement Component Object Reference=Get CharacterMovement!(self:Character Object Reference=@n7) -> (CharacterMovement:Character Movement Component Object Reference)) -> (ReturnValue:Vector)
   n9 = Get Actor Rotation(self=@n7)
   n10 = BreakVector(InVec=vector * vector(A=Get Rotation X Vector!(InRot:Rotator=@n9) -> (ReturnValue:Vector) {is=title K2Node_CallFunction}, B=@n8))
   n11 = Set PlayerVector_FwdBwd(execute=@n6.then_4, PlayerVector_FwdBwd:double=float + float(A=@n10, B=@n10.Y) {is=title K2Node_CommutativeAssociativeBinaryOperator}) -> (Output_Get:double)
   n12 = BreakVector(InVec=vector * vector(A=GetAxes!(A:Rotator=@n9) -> (X:Vector, Y:Vector, Z:Vector).Y, B=@n8))
   > Set PlayerVector_LR(PlayerVector_LR:double=float + float(A=@n12, B=@n12.Y)) -> (Output_Get:double)
   > Set relax_speed(relax_speed:double=MapRangeClamped(Value=@n11, InRangeA=0, InRangeB=-750, OutRangeA=1.75, OutRangeB=2.5)) -> (Output_Get:double)
   n13 = Branch(execute=@n6.then_5, Condition=Get bIsDBNO())
   n14 = GetPartSkeletalMeshComponent()
   > SetClothMaxDistanceScale!(execute, self:Skeletal Mesh Component Object Reference=@n14, Scale:float=3) -> (then)
   > Set is_dbno_alpha(is_dbno_alpha:double=0) -> (Output_Get:double)
   > Set dbno_trail_alpha(dbno_trail_alpha:double=MapRangeClamped(Value=Get PlayerVector_FwdBwd(), InRangeA=0, InRangeB=-750, OutRangeA=1, OutRangeB=0.3)) -> (Output_Get:double)
   SetClothMaxDistanceScale(execute=@n13.else, self=@n14, Scale=1)
   > Set is_dbno_alpha(is_dbno_alpha=1)
   n15 = Get PawnVelocityZ()
   n16 = "float > float"(A=@n15, B=0) {is=title K2Node_CallFunction}
   n17 = Get bIsSkydiveFloating()
   n18 = Get bIsParachuteOpened()
   Branch(execute=@n6.then_6, Condition=OR Boolean(A=AND Boolean(A=@n17, B=@n16), B=@n18))
   > EnableGravityOverride!(
    execute
    self:Clothing Simulation Interactor Object Reference=Get clothing_sim_interactor() -> (clothing_sim_interactor:Clothing Simulation Interactor Nv Object Reference)
    InVector_X:double=0
    InVector_Y:double=0
    InVector_Z:double=SelectFloat(A=-10000, B=float * float(A=@n15, B=10), bPickA=@n18)
   ) -> (then)
   n19 = GetAnimInstance(self=GetOwnerSkeletalMeshComponent())
   Is Valid(exec=@n6.then_7, InputObject=@n19)
   n20 = Montage_IsActive!(
    self:Anim Instance Object Reference=@n19
    Montage:Anim Montage Object Reference=/Figment_S10/Animation/Game/MainPlayer/Emotes/BlackMonday_Male/BlackMonday_Male_CMM_Montage
   ) -> (ReturnValue:bool)
   > n21 = Branch(Condition=@n20)
   > Set wind_blowing_left(wind_blowing_left:bool=false) -> (Output_Get:bool)
   > Set emote_mult(emote_mult:double=-3) -> (Output_Get:double)
   Set wind_blowing_left(execute=@n21.else, wind_blowing_left=true)
   > Set emote_mult(emote_mult=1)
   Branch(execute=@n6.then_8, Condition=Get bIsSkyDiving())
   n22 = MapRangeClamped(Value=Get PawnVelocityZ(), InRangeA=0, InRangeB=100, OutRangeA=1, OutRangeB=0.1)
   n23 = Get bIsSkydiveFloating()
   n24 = Get PlayerVector_FwdBwd()
   > Set skydive_trail_alpha(skydive_trail_alpha:double=float * float(A=MapRangeClamped(Value=@n24, InRangeA=0, InRangeB=-750, OutRangeA=1, OutRangeB=0.3), B=SelectFloat(A=@n22, B=1, bPickA=@n23))) -> (Output_Get:double)
   > Set skydive_constraint_alpha(skydive_constraint_alpha:double=SelectFloat(A=0.5, B=0, bPickA=Get bIsSkyDiving())) -> (Output_Get:double)
   n25 = Get player_pawn()
   n26 = GetVelocity!(self:Actor Object Reference=@n25) -> (ReturnValue:Vector)
   n27 = Get Actor Rotation(self=@n25)
   n28 = BreakVector(InVec=vector * vector(A=Get Rotation X Vector(InRot=@n27), B=@n26))
   n29 = float + float(A=@n28, B=@n28.Y)
   n30 = SelectFloat(A=0, B=1, bPickA="Not Equal (Float)"!(A:double=BreakVector(InVec=@n26).Z, B:double=0) -> (ReturnValue:bool) {is=title K2Node_CallFunction})
   n31 = Set SprintSpeedRemap(execute=@n6.then_3, SprintSpeedRemap:double=float * float(A=MapRangeClamped(Value=@n29, InRangeA=415, InRangeB=550), B=@n30)) -> (Output_Get:double)
   > n32 = Set RunSpeedRemap(RunSpeedRemap:double=float * float(A=MapRangeClamped(Value=@n29, InRangeA=0, InRangeB=300), B=@n30)) -> (Output_Get:double)
   n33 = AND Boolean(A=Get bIsBackpedaling() -> (bIsBackpedaling:bool), B=Get bIsMoving2D())
   > Set shoulder_corrective_alpha(shoulder_corrective_alpha:double="Max (Float)"!(A=@n31, B=@n32, C:double="To Float (Boolean)"(InBool=@n33)) -> (ReturnValue)) -> (Output_Get:double)
   n34 = Get wind_blowing_left() -> (wind_blowing_left:bool)
   n35 = SelectFloat(A=1.5, B=0.25, bPickA=@n34)
   n36 = float * float(A=Get emote_mult() -> (emote_mult:double), B=MapRangeClamped(Value=Get PawnSpeed2D(), InRangeA=0, InRangeB=200, OutRangeA=1, OutRangeB=0))
   n37 = GetGameTimeInSeconds()
   n38 = SelectFloat(A=10, B=1, bPickA=Get bIsSkyDiving())
   n39 = "Sin (Degrees)"!(A:double=float * float(A=float * float(A=@n37, B=25), B=@n38)) -> (ReturnValue:double)
   n40 = "Sin (Degrees)"(A=float * float(A=float * float(A=@n37, B=250), B=@n38))
   n41 = "Absolute (Float)"(A=float + float(A=float * float(A=@n40, B=20), B=float * float(A=@n39, B=80)))
   n42 = Get bIsCrouching()
   n43 = "Absolute (Float)"(A=Get!() -> (PawnDeltaRotation_Roll:float, PawnDeltaRotation_Pitch:float, PawnDeltaRotation_Yaw:float).PawnDeltaRotation_Yaw)
   n44 = vector * float(
    A=vector * vector!(A="-1,2,2", B_X:double=@n41, B_Y:double=@n41, B_Z:double=1) -> (ReturnValue)
    B=float * float(A=SelectFloat(A=0, B=1, bPickA=@n42), B=MapRangeClamped(Value=@n43, InRangeA=0, InRangeB=2, OutRangeA=1, OutRangeB=0))
   )
   Set fake_velocity_A(execute=@n6.then_1, fake_velocity_A:Vector=vector * float(A=@n44, B=float * float(A=@n35, B=@n36))) -> (Output_Get:Vector)
   n45 = SelectFloat(A=0.75, B=0.5, bPickA=@n34)
   > Set fake_velocity_B(fake_velocity_B:Vector=vector * float(A=@n44, B=float * float(A=@n45, B=@n36))) -> (Output_Get:Vector)
   n46 = SelectFloat(A=0.75, B=1, bPickA=@n34)
   > Set fake_velocity_C(fake_velocity_C:Vector=vector * float(A=@n44, B=float * float(A=@n46, B=@n36))) -> (Output_Get:Vector)
   n47 = SelectFloat(A=0.25, B=1.5, bPickA=@n34)
   > Set fake_velocity_D(fake_velocity_D:Vector=vector * float(A=@n44, B=float * float(A=@n47, B=@n36))) -> (Output_Get:Vector)
   n48 = Get Dynamic Material Instance(execute=@n6.then_2, SkeletalMeshComponent=Get face_mesh() -> (face_mesh:Skeletal Mesh Component Object Reference))
   > SetScalarParameterValue(self=@n48, ParameterName=Use Vertex Colors for Mask, Value=1)
   > SetScalarParameterValue(self=@n48, ParameterName=Hide Element 01, Value=1)
   Event BlueprintInitializeAnimation()
   n49 = TryGetPawnOwner()
   > Is Valid(InputObject=@n49)
   > n50 = Cast To FortPlayerPawnAthena(Object=@n49)
   > Set player_pawn(player_pawn:Fort Player Pawn Athena Object Reference=@n50) -> (Output_Get:Fort Player Pawn Athena Object Reference)
   > n51 = Cast To ClothingSimulationInteractorNv!(
    execute
    Object=GetClothingSimulationInteractor!(self:Skeletal Mesh Component Object Reference=GetPartSkeletalMeshComponent(), ClothingAssetName:Name=None) -> (ReturnValue:Clothing Simulation Interactor Object Reference)
   ) -> (then, CastFailed, AsClothing Simulation Interactor Nv:Clothing Simulation Interactor Nv Object Reference) {is=title K2Node_DynamicCast}
   > Set clothing_sim_interactor(clothing_sim_interactor:Clothing Simulation Interactor Nv Object Reference=@n51) -> (Output_Get:Clothing Simulation Interactor Nv Object Reference)
   Event OnEndSkydiving()
   > Set DynamicsAlpha(DynamicsAlpha:double=0.5) -> (Output_Get:double)
   Event OnBeginSkydiving()
   > Set DynamicsAlpha(DynamicsAlpha=1)
   Reroute Node()
   n52 = GetOwnerSkeletalMeshComponent()
   n53 = BreakRotator(InRot=BreakTransform(InTransform=GetDeltaTransformFromRefPose(self=@n52, BoneName=spine_05, BaseName=spine_01)).Rotation)
   n54 = MapRangeClamped(Value=@n53.Yaw, InRangeA=1, InRangeB=20)
   > Set spine_bent_back(spine_bent_back:double="Max (Float)"!(A=MapRangeClamped(Value=@n53.Yaw, InRangeA=-5, InRangeB=-30), B=@n54) -> (ReturnValue)) -> (Output_Get:double)
   > Set spine_bent_fwd(spine_bent_fwd:double=@n54) -> (Output_Get:double)
   Reroute Node()
   n55 = GetWorldDeltaSeconds!() -> (ReturnValue:double)
   n56 = GetDeltaTransformFromRefPose!(self=GetOwnerSkeletalMeshComponent(), BoneName=upperarm_r, BaseName=clavicle_r) -> (
    ReturnValue_Location:Vector
    ReturnValue_Rotation_Roll:float
    ReturnValue_Rotation_Pitch:float
    ReturnValue_Rotation_Yaw:float
    ReturnValue_Scale:Vector
   )
   > Set arm_roll_back_r(arm_roll_back_r:double=FInterpTo(
     Current=Get arm_roll_back_r() -> (arm_roll_back_r:double)
     Target=MapRangeClamped(Value=@n56.ReturnValue_Rotation_Roll, InRangeA=0, InRangeB=-25)
     DeltaTime=@n55
     InterpSpeed=20
    )) -> (Output_Get:double)
   > Set arm_roll_fwd_r(arm_roll_fwd_r:double=FInterpTo(
     Current=Get arm_roll_fwd_r() -> (arm_roll_fwd_r:double)
     Target=MapRangeClamped(Value=@n56.ReturnValue_Rotation_Roll, InRangeA=0, InRangeB=35)
     DeltaTime=@n55
     InterpSpeed=20
    )) -> (Output_Get:double)
   > Set arm_up_r(arm_up_r:double=MapRangeClamped(Value=@n56.ReturnValue_Rotation_Yaw, InRangeA=0, InRangeB=130, OutRangeB=0.75)) -> (Output_Get:double)
   > Set arm_pitch_up_r(arm_pitch_up_r:double=MapRangeClamped(Value=@n56.ReturnValue_Rotation_Pitch, InRangeA=0, InRangeB=70)) -> (Output_Get:double)
   n57 = GetDeltaTransformFromRefPose(self=GetOwnerSkeletalMeshComponent(), BoneName=upperarm_l, BaseName=clavicle_l)
   n58 = GetWorldDeltaSeconds()
   > Set arm_roll_back_l(arm_roll_back_l:double=FInterpTo(
     Current=Get arm_roll_back_l() -> (arm_roll_back_l:double)
     Target=MapRangeClamped(Value=@n57.ReturnValue_Rotation_Roll, InRangeA=0, InRangeB=-20)
     DeltaTime=@n58
     InterpSpeed=25
    )) -> (Output_Get:double)
   > Set arm_roll_fwd_l(arm_roll_fwd_l:double=FInterpTo(
     Current=Get arm_roll_fwd_l() -> (arm_roll_fwd_l:double)
     Target=MapRangeClamped(Value=@n57.ReturnValue_Rotation_Roll, InRangeA=0, InRangeB=-35)
     DeltaTime=@n58
     InterpSpeed=25
    )) -> (Output_Get:double)
   > Set arm_pitch_up_l(arm_pitch_up_l:double=MapRangeClamped(Value=@n57.ReturnValue_Rotation_Pitch, InRangeA=60, InRangeB=116)) -> (Output_Get:double)
   Reroute Node()
   n59 = GetPartSkeletalMeshComponent()
   n60 = GetDeltaTransformFromRefPose(self=@n59, BoneName=clavicle_r, BaseName=spine_05)
   > Set clav_upDown_r(clav_upDown_r:double=MapRangeClamped(Value=@n60.ReturnValue_Rotation_Pitch, InRangeA=1, InRangeB=20)) -> (Output_Get:double)
   > Set clav_leftRight_r(clav_leftRight_r:double=MapRangeClamped(Value=@n60.ReturnValue_Rotation_Roll, InRangeA=1, InRangeB=20)) -> (Output_Get:double)
   > Set clav_upDown_l(clav_upDown_l:double=MapRangeClamped(Value=GetDeltaTransformFromRefPose(self=@n59, BoneName=clavicle_l, BaseName=spine_05).ReturnValue_Rotation_Pitch, InRangeA=1, InRangeB=20)) -> (Output_Get:double)
   SelectFloat(A=-2, B=0.5, bPickA=@n20)
   vector + vector!(A="0, 0, 0", B_X:double=0, B_Y:double=0, B_Z:double=0) -> (ReturnValue)
   Reroute Node(InputPin:Wildcard) -> (OutputPin:Wildcard)
  ]}
]
variables [
 arm_pitch_up_l {type=number}
 arm_pitch_up_r {type=number}
 arm_roll_back_l {type=number}
 arm_roll_back_r {type=number}
 arm_roll_fwd_l {type=number}
 arm_roll_fwd_r {type=number}
 arm_up_r {type=number}
 body_mesh {type=SkeletalMeshComponent}
 clav_leftRight_r {type=number}
 clav_upDown_l {type=number}
 clav_upDown_r {type=number}
 clothing_sim_interactor {type=ClothingSimulationInteractorNv, default=null}
 dbno_trail_alpha {type=number}
 DynamicsAlpha {type=number}
 emote_mult {type=number, default=1}
 face_mesh {type=SkeletalMeshComponent}
 fake_velocity_A {type=Vector}
 fake_velocity_B {type=Vector}
 fake_velocity_back {type=Vector}
 fake_velocity_C {type=Vector}
 fake_velocity_D {type=Vector}
 fake_velocity_left {type=Vector}
 fake_velocity_right {type=Vector}
 is_dbno_alpha {type=number}
 player_pawn {type=FortPlayerPawnAthena}
 PlayerVector_FwdBwd {type=number}
 PlayerVector_LR {type=number}
 relax_speed {type=number, default=1}
 RunSpeedRemap {type=number}
 shoulder_corrective_alpha {type=number}
 skydive_constraint_alpha {type=number}
 skydive_trail_alpha {type=number, default=1}
 spine_bend_offset {type=Rotator}
 spine_bent_back {type=number}
 spine_bent_fwd {type=number}
 SprintSpeedRemap {type=number}
 wind_blowing_left {type=boolean, default=true}
]
